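; Impostor Baker Blueprint library that converts between 2D hemi-octahedral coordinates and 3D unit vectors.
; HemiOctahedronToUnitVector rotates and scales the input into X and Y, derives Z as one minus their absolute sum, then normalizes.
; UnitVectorToHemiOctahedron divides the absolute components by their sum and outputs the sum and difference of X and Y.
; The encoder discards component signs, so it inverts the decoder only for non-negative X, Y and Z.
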(fn HemiOctahedronToUnitVector
  (params (Hemi%20Octahedron :type Vector%202D%20Structure))
  (bind (_x _y) (Math|Vector2D|BreakVector2D Hemi%20Octahedron))
  (bind (_x_2 _y_2) (Math|Vector2D|BreakVector2D (* (Math|Vector2D|MakeVector2D (+ _x _y) (- _x _y)) 0.5)))
  (return (Math|Vector|Normalize (Math|Vector|MakeVector _x_2 _y_2 (- 1 (Math|Vector2D|DotProduct(2D) (Math|Vector2D|MakeVector2D (Math|Float|Absolute(Float) _x_2) (Math|Float|Absolute(Float) _y_2)) "(X=1.000000,Y=1.000000)"))))))

(fn UnitVectorToHemiOctahedron
  (params (Unit%20Vector :type Vector))
  (return (Math|Vector2D|MakeVector2D (+ (.x (/ (Math|Vector|MakeVector (Math|Float|Absolute(Float) (.x Unit%20Vector)) (Math|Float|Absolute(Float) (.y Unit%20Vector)) (Math|Float|Absolute(Float) (.z Unit%20Vector))) (Math|Vector|DotProduct (Math|Vector|MakeVector (Math|Float|Absolute(Float) (.x Unit%20Vector)) (Math|Float|Absolute(Float) (.y Unit%20Vector)) (Math|Float|Absolute(Float) (.z Unit%20Vector))) "1.000000,1.000000,1.000000"))) (.y (/ (Math|Vector|MakeVector (Math|Float|Absolute(Float) (.x Unit%20Vector)) (Math|Float|Absolute(Float) (.y Unit%20Vector)) (Math|Float|Absolute(Float) (.z Unit%20Vector))) (Math|Vector|DotProduct (Math|Vector|MakeVector (Math|Float|Absolute(Float) (.x Unit%20Vector)) (Math|Float|Absolute(Float) (.y Unit%20Vector)) (Math|Float|Absolute(Float) (.z Unit%20Vector))) "1.000000,1.000000,1.000000")))) (- (.x (/ (Math|Vector|MakeVector (Math|Float|Absolute(Float) (.x Unit%20Vector)) (Math|Float|Absolute(Float) (.y Unit%20Vector)) (Math|Float|Absolute(Float) (.z Unit%20Vector))) (Math|Vector|DotProduct (Math|Vector|MakeVector (Math|Float|Absolute(Float) (.x Unit%20Vector)) (Math|Float|Absolute(Float) (.y Unit%20Vector)) (Math|Float|Absolute(Float) (.z Unit%20Vector))) "1.000000,1.000000,1.000000"))) (.y (/ (Math|Vector|MakeVector (Math|Float|Absolute(Float) (.x Unit%20Vector)) (Math|Float|Absolute(Float) (.y Unit%20Vector)) (Math|Float|Absolute(Float) (.z Unit%20Vector))) (Math|Vector|DotProduct (Math|Vector|MakeVector (Math|Float|Absolute(Float) (.x Unit%20Vector)) (Math|Float|Absolute(Float) (.y Unit%20Vector)) (Math|Float|Absolute(Float) (.z Unit%20Vector))) "1.000000,1.000000,1.000000")))))))
